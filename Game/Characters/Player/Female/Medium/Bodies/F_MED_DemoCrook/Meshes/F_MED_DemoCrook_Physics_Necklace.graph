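# Physics asset for the necklace on the medium female DemoCrook player body, giving capsule bodies to the necklace and upper spine bones.
# Two dyn_necklace bodies chain from MSTR_necklace with small swing limits and zero twist limit, while MSTR_necklace joins spine_05 with free swing and twist.
# Larger capsules on spine_03 to spine_05, angularly locked to one another, form a rigid torso volume, apparently for the necklace to collide against.
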PhysicsAsset /Game/Characters/Player/Female/Medium/Bodies/F_MED_DemoCrook/Meshes/F_MED_DemoCrook_Physics_Necklace
bodies {
 MSTR_necklace [{shapeType=Capsule, length=3, radius=1, shapeName=MSTR_necklace_capsule}]
 dyn_necklace_01 [{shapeType=Capsule, length=5, radius=5, shapeName=dyn_necklace_01_capsule, center=<5.9977 0 5.192>, rotation=<rot -89.5454 0 0>}]
 dyn_necklace_02 [{shapeType=Capsule, length=4.5, radius=4, shapeName=dyn_necklace_02_capsule, center=<-0.5124 1.09 1.2354>, rotation=<rot -61.688 -180 180>}]
 spine_03 [{shapeType=Capsule, length=11.5, radius=9, shapeName=spine_03_capsule, center=<2.6075 2.1913 0>}]
 spine_04 [{shapeType=Capsule, length=9, radius=10, shapeName=spine_04_capsule, center=<0.6475 0.9152 0>}]
 spine_05 [{shapeType=Capsule, length=6.5, radius=8.5, shapeName=spine_05_capsule, center=<0.7365 -0.3577 0>}]
}
constraints [
 {
  bone1Name=MSTR_necklace
  bone2Name=spine_05
  swing1LimitDegrees=45
  swing1Motion=Free
  swing2LimitDegrees=45
  swing2Motion=Free
  twistLimitDegrees=0
  twistMotion=Free
 }
 {bone1Name=dyn_necklace_01, bone2Name=MSTR_necklace, swing1LimitDegrees=5, swing2LimitDegrees=20, twistLimitDegrees=0}
 {bone1Name=dyn_necklace_02, bone2Name=dyn_necklace_01, swing1LimitDegrees=15, swing2LimitDegrees=20, twistLimitDegrees=0}
 {
  bone1Name=spine_05
  bone2Name=spine_04
  swing1LimitDegrees=45
  swing1Motion=Locked
  swing2LimitDegrees=45
  swing2Motion=Locked
  twistLimitDegrees=0
  twistMotion=Locked
 }
 {
  bone1Name=spine_04
  bone2Name=spine_03
  swing1LimitDegrees=45
  swing1Motion=Locked
  swing2LimitDegrees=45
  swing2Motion=Locked
  twistLimitDegrees=0
  twistMotion=Locked
 }
]
