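# Control Rig for the DemoCrook female medium body's face accessory part, a corrective that poses hair bones, including the left pigtail root, from head orientation.
# A SphericalPoseReader on the head bone, with debug drawing on, outputs the weight for one ModifyTransforms step that applies preset offsets.
# The skeleton also has bangs, helmet, strap and pigtail dyn_ chains; the pigtail chains hang below the CR_ bones, and no simulation runs here.
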ControlRigBlueprint /Game/Characters/Player/Female/Medium/Bodies/F_MED_DemoCrook/Meshes/Parts/F_MED_DemoCrook_FaceAcc_CtrlRig
graphs [
 RigVMFunctionLibrary {links [], nodes []}
 RigVMModel {nodes [
   RigUnit_BeginExecution = &RigUnit_BeginExecution()
   n1 = &RigUnit_SphericalPoseReader(
    ActiveRegionScaleFactors=(PositiveWidth=3,NegativeWidth=2.5,PositiveHeight=2.5,NegativeHeight=2.5)
    ActiveRegionSize=0.12
    Debug=(bDrawDebug=True,bDraw2D=False,bDrawLocalAxes=False,DebugScale=25,DebugSegments=20,DebugThickness=0.25)
    DriverAxis=(X=1,Y=0,Z=0)
    DriverItem=(Type=Bone,Name="head")
    FalloffRegionScaleFactors=(PositiveWidth=1.5,NegativeWidth=1,PositiveHeight=1,NegativeHeight=1)
    FalloffSize=0.45
    FlipHeightScaling=False
    FlipWidthScaling=False
    OptionalParentItem=(Type=Bone,Name="None")
    RotationOffset=(X=-90,Y=0,Z=10)
    ExecutePin=@RigUnit_BeginExecution.ExecutePin
   )
   > &RigUnit_ModifyTransforms(
    ItemToModify="((Item=(Type=Bone,Name=\"None\"),Transform=(Rotation=(X=0,Y=0,Z=0,W=1),Translation=(X=0,Y=0,Z=0),Scale3D=(X=1,Y=1,Z=1))),(Item=(Type=Bone,Name=\"CR_pigTail_01_l\"),Transform=(Rotation=(X=-0.0057,Y=-0.0854,Z=-0.1812,W=0.9797),Translation=(X=-2.0484,Y=0.962,Z=-0 ...(truncated)"
    WeightMinimum=0
    Weight=@n1.OutputParam
   )
  ]}
]
hierarchy [
 root {loc=<0 0 0>, rot=<0 0 0 1>}
 pelvis {parent=root, loc=<0 -0.5301 98.0571>, rot=<0.0127 0.707 -0.0127 -0.707>}
 spine_01 {parent=pelvis, loc=<6.1208 -0.9328 0>, rot=<0 0 0.034 -0.9994>}
 spine_02 {parent=spine_01, loc=<6.5996 -0.2095 0>, rot=<0 0 0.0238 -0.9997>}
 spine_03 {parent=spine_02, loc=<6.4914 -0.1237 0>, rot=<0 0 0.0149 -0.9999>}
 spine_04 {parent=spine_03, loc=<6.9598 0.0794 0>, rot=<0 0 -0.0132 -0.9999>}
 spine_05 {parent=spine_04, loc=<6.9987 0 0>, rot=<0 0 -0.0015 -1>}
 neck_01 {parent=spine_05, loc=<11.3138 0.1529 0>, rot=<0 0 -0.2003 -0.9797>}
 neck_02 {parent=neck_01, loc=<4.5627 0 0>, rot=<0 0 0.0504 -0.9987>}
 head {parent=neck_02, loc=<5.3914 0 0>, rot=<0 0 0.115 -0.9934>}
 MSTR_bangsA {parent=head, loc=<9.9206 7.641 1.4937>, rot=<-0.0048 -0.1979 0.9493 0.2441>}
 dyn_bangsA_01 {parent=MSTR_bangsA, loc=<0 0 0>, rot=<0 0 0 1>}
 MSTR_bangsB {parent=head, loc=<10.2452 8.1641 0.7191>, rot=<-0.1389 -0.1501 0.9642 0.1687>}
 dyn_bangsB_01 {parent=MSTR_bangsB, loc=<0 0 0>, rot=<0 0 0 1>}
 MSTR_helmet {parent=head, loc=<0 0 0>, rot=<0 0 0 1>}
 dyn_helmet {parent=MSTR_helmet, loc=<0 0 0>, rot=<0 0 0 1>}
 MSTR_strap_l {parent=dyn_helmet, loc=<5.3393 2.9568 -7.1973>, rot=<0.0305 0.6431 0.7627 0.0621>}
 dyn_strap_01_l {parent=MSTR_strap_l, loc=<0 0 0>, rot=<0 0 0 1>}
 dyn_strap_02_l {parent=dyn_strap_01_l, loc=<3.9737 0 0>, rot=<0 0 0 1>}
 dyn_strap_03_l {parent=dyn_strap_02_l, loc=<3.5953 0 0>, rot=<0 0 0 1>}
 dyn_strap_04_l {parent=dyn_strap_03_l, loc=<3.1629 0 0>, rot=<0 0 0 1>}
 MSTR_strap_r {parent=dyn_helmet, loc=<5.3393 2.9568 7.1973>, rot=<-0.0305 -0.6431 0.7627 0.0621>}
 dyn_strap_01_r {parent=MSTR_strap_r, loc=<0 0 0>, rot=<0 0 0 1>}
 dyn_strap_02_r {parent=dyn_strap_01_r, loc=<3.9737 0 0>, rot=<0 0 0 1>}
 dyn_strap_03_r {parent=dyn_strap_02_r, loc=<3.5953 0 0>, rot=<0 0 0 1>}
 dyn_strap_04_r {parent=dyn_strap_03_r, loc=<3.1629 0 0>, rot=<0 0 0 1>}
 CR_pigTail_01_l {parent=head, loc=<0.6599 -5.0401 -3.9552>, rot=<0.4073 -0.7744 -0.4603 -0.1503>}
 CR_pigTail_02_l {parent=CR_pigTail_01_l, loc=<2.429 0 0>, rot=<0 0 0 1>}
 MSTR_pigTail_End_l {parent=CR_pigTail_02_l, loc=<1.6527 0 0>, rot=<0 0 0 1>}
 dyn_pigTail_01_l {parent=MSTR_pigTail_End_l, loc=<0 0 0>, rot=<0 0 0 1>}
 dyn_pigTail_02_l {parent=dyn_pigTail_01_l, loc=<2.2548 0 0>, rot=<0 0 0 1>}
 dyn_pigTail_03_l {parent=dyn_pigTail_02_l, loc=<2.3326 0 0>, rot=<0 0 0 1>}
 CR_pigTail_01_r {parent=head, loc=<0.7328 -4.9569 4.0978>, rot=<0.4078 -0.7758 0.4579 0.1491>}
 CR_pigTail_02_r {parent=CR_pigTail_01_r, loc=<2.4288 0 0>, rot=<0 0 0 -1>}
 MSTR_pigTail_End_r {parent=CR_pigTail_02_r, loc=<2.146 0 0>, rot=<0 0 0 1>}
 dyn_pigTail_01_r {parent=MSTR_pigTail_End_r, loc=<0 0 0>, rot=<0 0 0 1>}
 dyn_pigTail_02_r {parent=dyn_pigTail_01_r, loc=<1.8805 0 0>, rot=<0 0 0 1>}
 dyn_pigTail_03_r {parent=dyn_pigTail_02_r, loc=<2.2365 0 0>, rot=<0 0 0 1>}
]
