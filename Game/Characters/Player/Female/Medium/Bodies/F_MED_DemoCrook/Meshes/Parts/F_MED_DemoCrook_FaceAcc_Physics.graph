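# Physics asset for the face accessory part of the medium female DemoCrook player body, covering left and right pigtails and straps.
# Each of the four chains has a capsule body on an MSTR root bone, followed by three capsule-bodied dyn bones constrained in sequence.
# Constraints allow limited swing with zero twist, with pigtails held tighter than straps.
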PhysicsAsset /Game/Characters/Player/Female/Medium/Bodies/F_MED_DemoCrook/Meshes/Parts/F_MED_DemoCrook_FaceAcc_Physics
bodies {
 MSTR_pigTail_End_l [{shapeType=Capsule, length=5, radius=4, shapeName=MSTR_pigTail_End_l_capsule, center=<3.7 0 0.4>, rotation=<rot -55 0 0>}]
 MSTR_pigTail_End_r [{shapeType=Capsule, length=5, radius=4, shapeName=MSTR_pigTail_End_r_capsule, center=<4.6 0 1>, rotation=<rot 55 0 0>}]
 MSTR_strap_l [{shapeType=Capsule, length=3, radius=2, shapeName=MSTR_strap_l_capsule}]
 MSTR_strap_r [{shapeType=Capsule, length=3, radius=2, shapeName=MSTR_strap_r_capsule}]
 dyn_pigTail_01_l [{shapeType=Capsule, length=2, radius=1, shapeName=dyn_pigTail_01_l_capsule, center=<3.2615 0 0>, rotation=<rot -90 0 0>}]
 dyn_pigTail_01_r [{shapeType=Capsule, length=2, radius=1, shapeName=dyn_pigTail_01_r_capsule, center=<3.2615 0 0>, rotation=<rot -90 0 0>}]
 dyn_pigTail_02_l [{shapeType=Capsule, length=2, radius=1, shapeName=dyn_pigTail_02_l_capsule, center=<3.2615 0 0>, rotation=<rot -90 0 0>}]
 dyn_pigTail_02_r [{shapeType=Capsule, length=2, radius=1, shapeName=dyn_pigTail_02_r_capsule, center=<3.2615 0 0>, rotation=<rot -90 0 0>}]
 dyn_pigTail_03_l [{shapeType=Capsule, length=2, radius=1, shapeName=dyn_pigTail_03_l_capsule, center=<3.2615 0 0>, rotation=<rot -90 0 0>}]
 dyn_pigTail_03_r [{shapeType=Capsule, length=2, radius=1, shapeName=dyn_pigTail_03_r_capsule, center=<3.2615 0 0>, rotation=<rot -90 0 0>}]
 dyn_strap_01_l [{shapeType=Capsule, length=5, radius=3, shapeName=dyn_strap_01_l_capsule, center=<6 0 1.8>, rotation=<rot -90 0 0>}]
 dyn_strap_01_r [{shapeType=Capsule, length=5, radius=3, shapeName=dyn_strap_01_r_capsule, center=<6 0 -1.8>, rotation=<rot -90 0 0>}]
 dyn_strap_02_l [{shapeType=Capsule, length=5, radius=3, shapeName=dyn_strap_02_l_capsule, center=<6 0 1.8>, rotation=<rot -90 0 0>}]
 dyn_strap_02_r [{shapeType=Capsule, length=5, radius=3, shapeName=dyn_strap_02_r_capsule, center=<6 0 -1.8>, rotation=<rot -90 0 0>}]
 dyn_strap_03_l [{shapeType=Capsule, length=5, radius=3, shapeName=dyn_strap_03_l_capsule, center=<6 0 1.8>, rotation=<rot -90 0 0>}]
 dyn_strap_03_r [{shapeType=Capsule, length=5, radius=3, shapeName=dyn_strap_03_r_capsule, center=<6 0 -1.8>, rotation=<rot -90 0 0>}]
}
constraints [
 {bone1Name=dyn_strap_02_r, bone2Name=dyn_strap_01_r, swing1LimitDegrees=15, swing2LimitDegrees=10, twistLimitDegrees=0}
 {bone1Name=dyn_strap_03_r, bone2Name=dyn_strap_02_r, swing1LimitDegrees=15, swing2LimitDegrees=10, twistLimitDegrees=0}
 {bone1Name=dyn_strap_01_l, bone2Name=MSTR_strap_l, swing1LimitDegrees=20, swing2LimitDegrees=30, twistLimitDegrees=0}
 {bone1Name=dyn_strap_02_l, bone2Name=dyn_strap_01_l, swing1LimitDegrees=15, swing2LimitDegrees=10, twistLimitDegrees=0}
 {bone1Name=dyn_strap_03_l, bone2Name=dyn_strap_02_l, swing1LimitDegrees=15, swing2LimitDegrees=10, twistLimitDegrees=0}
 {bone1Name=dyn_strap_01_r, bone2Name=MSTR_strap_r, swing1LimitDegrees=20, swing2LimitDegrees=30, twistLimitDegrees=0}
 {bone1Name=dyn_pigTail_01_r, bone2Name=MSTR_pigTail_End_r, swing1LimitDegrees=5, swing2LimitDegrees=5, twistLimitDegrees=0}
 {bone1Name=dyn_pigTail_02_r, bone2Name=dyn_pigTail_01_r, swing1LimitDegrees=10, swing2LimitDegrees=8, twistLimitDegrees=0}
 {bone1Name=dyn_pigTail_03_r, bone2Name=dyn_pigTail_02_r, swing1LimitDegrees=10, swing2LimitDegrees=8, twistLimitDegrees=0}
 {bone1Name=dyn_pigTail_01_l, bone2Name=MSTR_pigTail_End_l, swing1LimitDegrees=5, swing2LimitDegrees=5, twistLimitDegrees=0}
 {bone1Name=dyn_pigTail_02_l, bone2Name=dyn_pigTail_01_l, swing1LimitDegrees=10, swing2LimitDegrees=8, twistLimitDegrees=0}
 {bone1Name=dyn_pigTail_03_l, bone2Name=dyn_pigTail_02_l, swing1LimitDegrees=10, swing2LimitDegrees=8, twistLimitDegrees=0}
]
